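# Physics asset for the Donut Dish back bling, letting its two gun pieces move physically on the backpack.
# A tiny capsule body on root serves as the hub for two larger capsules on the dyn_gun_l and dyn_gun_r bones.
# Each gun body is constrained to root with narrow swing limits and an even tighter twist limit, which restricts it to slight wobble.
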PhysicsAsset /BRCosmetics/Accessories/FORT_Backpacks/Backpack_F_MED_Donut_Dish/Meshes/F_MED_Donut_Dish_Physics
bodies {
 dyn_gun_l [{shapeType=Capsule, length=18.9875, radius=7.8553, center=<1.6602 0 -1.1904>, rotation=<rot 90 0 0>}]
 dyn_gun_r [{shapeType=Capsule, length=19.0488, radius=7.8553, center=<2.9367 0 0.7031>, rotation=<rot 90 0 0>}]
 root [{shapeType=Capsule, length=0.505, radius=0.505}]
}
constraints [
 {bone1Name=dyn_gun_l, bone2Name=root, swing1LimitDegrees=12, swing2LimitDegrees=12, twistLimitDegrees=3}
 {bone1Name=dyn_gun_r, bone2Name=root, swing1LimitDegrees=12, swing2LimitDegrees=12, twistLimitDegrees=3}
]
